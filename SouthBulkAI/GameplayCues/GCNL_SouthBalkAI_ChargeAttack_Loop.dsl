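; Looping GameplayCue notify that drives the South Bulk AI charge attack's charge, energy ray and impact effects.
; On loop start it caches the three particle components, places charge and ray at mesh-relative offsets and sets the ray's direction parameter.
; A repeating timer line-traces along the ray and moves the impact effect to the hit location, without checking for a hit.
; Removal deactivates only the impact effect and does not explicitly clear the timer.
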(event EventOnLoopingStart(EitherVFX-preferthisversion)
  (params (MyTarget :type Actor%20Object%20Reference) (Parameters :type Gameplay%20Cue%20Parameters%20Structure%20%28by%20ref%29) (ParticleComponents :type Array%20of%20FXSystem%20Component%20Object%20References) (AudioComponents :type Array%20of%20Audio%20Component%20Object%20References))
  (bind _energyrayfx (Variables|Default|GetEnergyRayFX))
  (|CallParentFunction MyTarget Parameters ParticleComponents AudioComponents)
  (Variables|Default|SetChargeFX (Utilities|Array|Get(acopy) ParticleComponents))
  (Variables|Default|SetEnergyRayFX (Utilities|Array|Get(acopy) ParticleComponents 1))
  (bind _output_get (Variables|Default|SetImpactFX (Utilities|Array|Get(acopy) ParticleComponents 2)))
  (Components|Activation|SetActive _output_get true false)
  (bind _asbp_player_pawn_non_participant_south_bulk_ai_v2 (Utilities|Casting|CastToBP_PlayerPawn_NonParticipant_SouthBulkAI_V2 MyTarget))
  (Utilities|FlowControl|Sequence
    (:then_0
      ;[0.09478 0.9583 0.2284 0.5] Set correct positions for FX (should be updated once the final model is implemented)
      (Transformation|SetWorldLocation (Variables|Default|GetChargeFX) (+ (Math|Vector|RotateVector (Variables|Default|GetChargeFXOrigin) (Transformation|GetWorldRotation (Variables|/Script/Engine.Character|GetMesh _asbp_player_pawn_non_participant_south_bulk_ai_v2))) (Transformation|GetWorldLocation (Variables|/Script/Engine.Character|GetMesh _asbp_player_pawn_non_participant_south_bulk_ai_v2))) false true)
      (Transformation|SetWorldLocation _energyrayfx (+ (Transformation|GetWorldLocation (Variables|/Script/Engine.Character|GetMesh _asbp_player_pawn_non_participant_south_bulk_ai_v2)) (Math|Vector|RotateVector (Variables|Default|GetLaserBeamOrigin) (Transformation|GetWorldRotation (Variables|/Script/Engine.Character|GetMesh _asbp_player_pawn_non_participant_south_bulk_ai_v2)))) false true)
      (Effects|Components|ParticleSystem|SetVectorParameter _energyrayfx "EnergyRayDirection" (Variables|Default|GetEnergyRayDirection)))
    (:then_1
      (Utilities|Time|SetTimerbyEvent (delegate Custom|GroundCheck) 0.15 true))))

(event EventOnRemoval(EitherVFX-preferthisversion)
  (params (MyTarget :type Actor%20Object%20Reference) (Parameters :type Gameplay%20Cue%20Parameters%20Structure%20%28by%20ref%29) (ParticleComponents :type Array%20of%20FXSystem%20Component%20Object%20References) (AudioComponents :type Array%20of%20Audio%20Component%20Object%20References) (BurstCameraShakeInstance :type Camera%20Shake%20Base%20Object%20Reference) (BurstDecalInstance :type Decal%20Actor%20Object%20Reference))
  (|CallParentFunction MyTarget Parameters ParticleComponents AudioComponents BurstCameraShakeInstance BurstDecalInstance)
  (Components|Activation|Deactivate (Variables|Default|GetImpactFX)))

(event Custom|GroundCheck
  ;[1 0.1098 0.1423 0.5] Set impact FX location
  (bind _outhit (Collision|LineTraceByChannel 0 (+ (Transformation|GetWorldLocation (Variables|/Script/Engine.Character|GetMesh (stale EventOnLoopingStart(EitherVFX-preferthisversion) _asbp_player_pawn_non_participant_south_bulk_ai_v2))) (Math|Vector|RotateVector (Variables|Default|GetLaserBeamOrigin) (Transformation|GetWorldRotation (Variables|/Script/Engine.Character|GetMesh (stale EventOnLoopingStart(EitherVFX-preferthisversion) _asbp_player_pawn_non_participant_south_bulk_ai_v2))))) (+ (+ (Transformation|GetWorldLocation (Variables|/Script/Engine.Character|GetMesh (stale EventOnLoopingStart(EitherVFX-preferthisversion) _asbp_player_pawn_non_participant_south_bulk_ai_v2))) (Math|Vector|RotateVector (Variables|Default|GetLaserBeamOrigin) (Transformation|GetWorldRotation (Variables|/Script/Engine.Character|GetMesh (stale EventOnLoopingStart(EitherVFX-preferthisversion) _asbp_player_pawn_non_participant_south_bulk_ai_v2))))) (* (Math|Vector|RotateVector (Variables|Default|GetEnergyRayDirection) (Transformation|GetWorldRotation (Variables|/Script/Engine.Character|GetMesh (stale EventOnLoopingStart(EitherVFX-preferthisversion) _asbp_player_pawn_non_participant_south_bulk_ai_v2)))) (Variables|Default|GetEnergyRayLength))) "TraceTypeQuery1" false (Utilities|Array|MakeArray (stale EventOnLoopingStart(EitherVFX-preferthisversion) _asbp_player_pawn_non_participant_south_bulk_ai_v2))))
  (Transformation|SetWorldLocation (Variables|Default|GetImpactFX) (.Location (Collision|BreakHitResult _outhit)) false true))

(fn ConstructionScript)
